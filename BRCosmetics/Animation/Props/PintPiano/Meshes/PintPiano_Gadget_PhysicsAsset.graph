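# Physics asset for the PintPiano gadget cosmetic prop, giving its skeleton simple collision and physics bodies.
# The controller and vehicle bones each carry a large capsule, while the main bone has only a tiny one.
# Constraints join both the vehicle and controller bodies directly to main, with matching swing and twist limit angles.
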PhysicsAsset /BRCosmetics/Animation/Props/PintPiano/Meshes/PintPiano_Gadget_PhysicsAsset
bodies {
 JNT_Controller [{shapeType=Capsule, length=24.8763, radius=12.9881, center=<-0.1487 -15.1969 2.0658>, rotation=<rot 0 170.4676 -86.6612>}]
 JNT_Vehicle [{shapeType=Capsule, length=30.9635, radius=21.4677, center=<-3.0451 1.4816 0.0703>, rotation=<rot 0 113.1828 -89.8013>}]
 main [{shapeType=Capsule, length=0.505, radius=0.505, rotation=<rot 0 -90 0>}]
}
constraints [
 {bone1Name=JNT_Vehicle, bone2Name=main, swing1LimitDegrees=45, swing2LimitDegrees=45, twistLimitDegrees=45}
 {bone1Name=JNT_Controller, bone2Name=main, swing1LimitDegrees=45, swing2LimitDegrees=45, twistLimitDegrees=45}
]
